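# Material function for the EmberRae backpack that layers an animated belly, throat and mouth glow onto incoming BaseColor and Emissive.
# UV sphere masks locate the glows, combined with an FX texture channel, sine-driven breathing, panning noise and a MaterialCurveA boost.
# Inner, mid and outer colors form the glow gradient, scaled by EmissiveColorSaturation.
# The blended emissive passes through the shared time-of-day color adjustment before output.
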MaterialFunction /BRCosmetics/Accessories/FORT_Backpacks/Backpack_F_MED_EmberRae/Materials/MF_EmberRae_Glow
graph {
 properties {description=GlowMaskPower}
 nodes [
  n1 = GetMaterialAttributes(None=FunctionInput() {inputType=FunctionInput_MaterialAttributes}) {attributeGetTypes [#1, #2]}
  n2 = NamedRerouteUsage()
  n3 = LinearInterpolate(A=@n1.EmissiveColor, B=@n2, Alpha=NamedRerouteUsage())
  n4 = NamedRerouteUsage()
  FunctionOutput(None=SetMaterialAttributes!(
    MaterialAttributes=@n1.MaterialAttributes
    EmissiveColor=MaterialFunctionCall!(Color=@n3) {materialFunction=/Game/Packages/Fortress_SharedMaterials/Anime_Master_Material/MF_TODColorAdjustment}.Result
    BaseColor=LinearInterpolate(A=@n1.BaseColor, B=@n2, Alpha=@n4)
   ) {attributeSetTypes [#1, #2]})
  n5 = ScalarParameter() {defaultValue=0.8, parameterName=InnerGlow Radius}
  n6 = Clamp(None=NamedRerouteUsage(), Min=@n5) {minDefault=0.5}
  n7 = ScalarParameter() {defaultValue=0.1, parameterName=MidGlow Radius}
  n8 = Subtract(A=Clamp(None=NamedRerouteUsage(), Min=@n7) {minDefault=0.5}, B=@n7) {constB=0.5}
  n9 = LinearInterpolate(
   A=Multiply(
    A=VectorParameter() {group=EmberRae, parameterName=InnerGlow Color, defaultValue=<rgba 1 0.9175 0 1>}.RGB
    B=Subtract(A=@n6, B=@n5) {constB=0.5}
   )
   B=Multiply(A=VectorParameter() {group=EmberRae, parameterName=MidGlow Color, defaultValue=<rgba 0.8073 0.1964 0 1>}.RGB, B=@n8)
   Alpha=@n8
  )
  n10 = NamedRerouteUsage()
  n11 = Multiply(A=VectorParameter() {group=EmberRae, parameterName=OuterGlow Color, defaultValue=<rgba 0.9323 0 0.0075 1>}.RGB, B=@n10)
  n12 = OneMinus(None=@n10)
  NamedRerouteDeclaration(None=Multiply(A=LinearInterpolate(A=@n9, B=@n11, Alpha=@n12), B=ScalarParameter() {defaultValue=2, group=EmberRae, parameterName=EmissiveColorSaturation}) {constB=2}) {name=outColor, nodeColor=<rgba 0.6875 1 0 1>}
  n13 = Time()
  NamedRerouteDeclaration(None=Frac(None=Multiply(A=@n13) {constB=0.25})) {name=LocalDebug, nodeColor=<rgba 0 0.1797 1 1>}
  n14 = ScalarParameter() {group=EmberRae, parameterName=MaterialCurveA}
  n15 = TextureSampleParameter2D() {
   group=EmberRae
   parameterName=FXTexture
   samplerType=SAMPLERTYPE_Masks
   texture=/BRCosmetics/Accessories/FORT_Backpacks/Backpack_F_MED_EmberRae/Textures/T_F_MED_EmberRae_Backpack_FX
  }
  n16 = SphereMask(
   A=TextureCoordinate()
   B=Constant2Vector() {g=0.14, r=0.6}
   Radius=Add(A=NamedRerouteUsage(), B=ScalarParameter() {defaultValue=0.06, group=EmberRae, parameterName=Belly Radius})
  ) {attenuationRadius=0.07, hardnessPercent=0}
  n17 = Multiply(A=Clamp(None=@n16) {maxDefault=0.9}, B=ScalarParameter() {defaultValue=1, group=EmberRae, parameterName=Belly Glow Opacity})
  n18 = TextureCoordinate()
  n19 = Constant2Vector() {g=0.5, r=0.215}
  n20 = Add(A=NamedRerouteUsage(), B=ScalarParameter() {defaultValue=0.065, group=EmberRae, parameterName=Throat Radius})
  n21 = Multiply(
   A=Clamp(None=SphereMask(A=@n18, B=@n19, Radius=@n20) {attenuationRadius=0.07, hardnessPercent=0}) {maxDefault=0.9}
   B=ScalarParameter() {defaultValue=0.65, group=EmberRae, parameterName=Throat Glow Opacity}
  )
  n22 = ScalarParameter() {defaultValue=0.08, group=EmberRae, parameterName=Mouth Radius}
  n23 = NamedRerouteUsage()
  n24 = Multiply(
   A=SphereMask(A=TextureCoordinate(), B=Constant2Vector() {g=0.1, r=0.93}, Radius=Add(A=@n22, B=@n23)) {attenuationRadius=0.07, hardnessPercent=0}
   B=NamedRerouteUsage()
  )
  n25 = ScalarParameter() {defaultValue=1, group=EmberRae, parameterName=Mouth Glow Opacity}
  n26 = Add(
   A=Add(
    A=NamedRerouteDeclaration(None=@n17) {name=Mask Belly, nodeColor=<rgba 0.7344 0 1 1>}
    B=NamedRerouteDeclaration(None=@n21) {name=Mask Throat, nodeColor=<rgba 0.7344 0 1 1>}
   )
   B=NamedRerouteDeclaration(None=Multiply(A=@n24, B=@n25)) {name=MouthMask, nodeColor=<rgba 0.7344 0 1 1>}
  )
  n27 = Multiply(
   A=NamedRerouteDeclaration(None=StaticComponentMaskParameter(None=@n15.RGB) {defaultR=true, group=EmberRae, parameterName=FXChannel}) {name=FX Mask, nodeColor=<rgba 1 0 0.2109 1>}
   B=Multiply(A=NamedRerouteUsage(), B=NamedRerouteDeclaration(None=@n26) {name=MaskCombine, nodeColor=<rgba 0.7344 0 1 1>})
  )
  NamedRerouteDeclaration(None=Add(A=Multiply(A=@n14, B=@n27), B=@n27)) {name=GlowMask, nodeColor=<rgba 0 0.7656 1 1>}
  n28 = TextureCoordinate()
  n29 = ScalarParameter() {defaultValue=1, parameterName=NoiseTiling}
  n30 = TextureSample(UVs=Panner(Coordinate=Multiply(A=@n28, B=@n29)) {speedY=0.2}) {texture=/Game/Global/Textures/Default/Noise/TankMoteNoise}
  n31 = ScalarParameter() {defaultValue=0.02, parameterName=NoiseScale}
  n32 = TextureCoordinate()
  n33 = ScalarParameter() {defaultValue=5, parameterName=NoiseTiling2}
  n34 = TextureSample(UVs=Panner(Coordinate=Multiply(A=@n32, B=@n33)) {speedY=0.1}) {samplerSource=SSM_Wrap_WorldGroupSettings, texture=/Game/Weapons/WeaponSkins/Wraps/Textures/RadishMem/T_Wrap_Radish_CloudNoise}
  n35 = ScalarParameter() {defaultValue=0.015, parameterName=NoiseScale2}
  NamedRerouteDeclaration(None=Add(A=Multiply(A=@n30.R, B=@n31), B=Multiply(A=@n34.R, B=@n35))) {name=Noise, nodeColor=<rgba 0 0.0625 1 1>}
  n36 = Sine(None=Multiply(A=ScalarParameter() {defaultValue=0.25, group=EmberRae, parameterName=Breathing Speed}, B=Time()))
  n37 = ScalarParameter() {defaultValue=0.6, group=EmberRae, parameterName=Breathing Strength}
  NamedRerouteDeclaration(None=Clamp(None=ConstantBiasScale(None=@n36), Min=OneMinus(None=@n37)) {minDefault=0.25}) {name=Breathing, nodeColor=<rgba 0.7344 0 1 1>}
  Power(Base=@n16) {constExponent=3}
  ScalarParameter() {defaultValue=1, group=EmberRae, parameterName=GlowMaskPower}
 ]
}
