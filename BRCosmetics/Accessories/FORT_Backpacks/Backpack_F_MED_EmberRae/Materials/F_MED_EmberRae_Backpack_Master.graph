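# Master character material for the EmberRae backpack cosmetic, built on material attributes for skeletal meshes with morph targets and mesh deformers.
# It chains shared functions: BaseFN_CharacterFunction_PreFX feeds CharacterSkinFX, then BaseFN_CharacterFunction_PostFX and CINE_MF_CharPostFX.
# The useEmberRaeEmissive static switch optionally inserts MF_EmberRae_Glow between the skin FX and PostFX stages.
# Shading model comes from the material expression, spanning lit, subsurface, clear coat, cloth and eye models.
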Material /BRCosmetics/Accessories/FORT_Backpacks/Backpack_F_MED_EmberRae/Materials/F_MED_EmberRae_Backpack_Master
graph {
 properties {
  bAutomaticallySetUsageInEditor=false
  bUseMaterialAttributes=true
  bUsedWithMeshDeformer=true
  bUsedWithMorphTargets=true
  bUsedWithSkeletalMesh=true
  pixelDepthOffsetMode=PDOM_Legacy
  refractionMethod=RM_IndexOfRefraction
  shadingModel=MSM_FromMaterialExpression
  subsurfaceProfile=&SS_HeroSkin_02
  usedShadingModels=Default Lit | Subsurface | Clear Coat | Subsurface Profile | Cloth | Eye
 }
 nodes [
  n1 = MaterialFunctionCall() {materialFunction=&BaseFN_CharacterFunction_PreFX}
  n2 = MaterialFunctionCall(
   PreskinnedPosition=@n1.Pre skinned Local Position
   Gmap=@n1.Pre skinned Local Position
   MaterialAttributes=@n1.Pre skinned Local Position
   AO=@n1.Pre skinned Local Position
  ) {materialFunction=&CharacterSkinFX}
  n3 = MaterialFunctionCall(
   AO=@n1.AO
   Distance Blend=@n1.AO
   Fresnel=@n1.AO
   Gmap=@n1.AO
   Pre Skinned Local Position=@n1.AO
   Material=StaticSwitchParameter(
    True=MaterialFunctionCall!(In=@n2.Result) {materialFunction=&/MF_EmberRae_Glow}.Result
    False=@n2.Result
   ) {group=EmberRae, parameterName=useEmberRaeEmissive}
  ) {materialFunction=&BaseFN_CharacterFunction_PostFX}
  MaterialFunctionCall(In=@n3.Result) {materialFunction=&CINE_MF_CharPostFX}
 ]
}
