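# Physics asset for the EmberRae backpack cosmetic, giving a small figure with arms, legs and a tail on the pack its physical bodies.
# Capsule bodies on the dyn_ bones form limb and three-segment tail chains branching from dyn_pelvis, whose rotation is locked to the pack bone.
# Elbows and knees lock twist and one swing axis to act as narrow hinges, while other joints have tight angle limits, narrowing toward the tail tip.
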PhysicsAsset /BRCosmetics/Accessories/FORT_Backpacks/Backpack_F_MED_EmberRae/Meshes/F_MED_EmberRae_Pack_Physics
bodies {
 dyn_calf_l [{shapeType=Capsule, length=2.5233, radius=2.0101, shapeName=dyn_calf_l_capsule, center=<1.8015 0 0>, rotation=<rot -90 180 180>}]
 dyn_calf_r [{shapeType=Capsule, length=2.5233, radius=2.0101, shapeName=dyn_calf_r_capsule, center=<-1.8015 0 0>, rotation=<rot -90 0 0>}]
 dyn_foot_l [{shapeType=Capsule, length=2.2706, radius=1.5, shapeName=dyn_foot_l_capsule, center=<1.6569 0 0>, rotation=<rot -90 180 180>}]
 dyn_foot_r [{shapeType=Capsule, length=2.2706, radius=1.5, shapeName=dyn_foot_r_capsule, center=<-1.6569 0 0>, rotation=<rot -90 0 0>}]
 dyn_hand_l [{shapeType=Capsule, length=1, radius=1.5, shapeName=dyn_hand_l_capsule, center=<1.3204 0 0>, rotation=<rot -90 180 180>}]
 dyn_hand_r [{shapeType=Capsule, length=1, radius=1.5, shapeName=dyn_hand_r_capsule, center=<-1.3204 0 0>, rotation=<rot -90 0 0>}]
 dyn_lowerarm_l [{shapeType=Capsule, length=1, radius=2, shapeName=dyn_lowerarm_l_capsule, center=<1.6668 0 0>, rotation=<rot -90 180 180>}]
 dyn_lowerarm_r [{shapeType=Capsule, length=1, radius=2, shapeName=dyn_lowerarm_r_capsule, center=<-1.6668 0 0>, rotation=<rot -90 0 0>}]
 dyn_pelvis [{shapeType=Capsule, length=5.0049, radius=6.0597, shapeName=dyn_pelvis_capsule, rotation=<rot 0 0 -90>}]
 dyn_tail_01 [{shapeType=Capsule, length=2, radius=3, shapeName=dyn_tail_01_capsule, center=<2.6975 0 0>, rotation=<rot -90 0 0>}]
 dyn_tail_02 [{shapeType=Capsule, length=1.5, radius=2, shapeName=dyn_tail_02_capsule, center=<2.1967 0 0>, rotation=<rot -90 180 180>}]
 dyn_tail_03 [{shapeType=Capsule, length=1.5, radius=1.5, shapeName=dyn_tail_03_capsule, center=<1.255 0 0>, rotation=<rot -90 180 180>}]
 dyn_thigh_l [{shapeType=Capsule, length=2.25, radius=3, shapeName=dyn_thigh_l_capsule, center=<3.4822 0 0>, rotation=<rot -90 0 0>}]
 dyn_thigh_r [{shapeType=Capsule, length=2.25, radius=3, shapeName=dyn_thigh_r_capsule, center=<-3.4822 0 0>, rotation=<rot -90 0 0>}]
 dyn_upperarm_l [{shapeType=Capsule, length=1, radius=2.5, shapeName=dyn_upperarm_l_capsule, center=<1.4853 0 0>, rotation=<rot -90 0 0>}]
 dyn_upperarm_r [{shapeType=Capsule, length=1, radius=2.5, shapeName=dyn_upperarm_r_capsule, center=<-1.4853 0 0>, rotation=<rot -90 0 0>}]
 pack [{shapeType=Capsule, length=0.505, radius=0.505, shapeName=pack_capsule, rotation=<rot 0 -90 0>}]
}
constraints [
 {bone1Name=dyn_upperarm_l, bone2Name=dyn_pelvis, swing1LimitDegrees=14, swing2LimitDegrees=14, twistLimitDegrees=10}
 {
  bone1Name=dyn_lowerarm_l
  bone2Name=dyn_upperarm_l
  swing1LimitDegrees=45
  swing1Motion=Locked
  swing2LimitDegrees=10
  twistLimitDegrees=45
  twistMotion=Locked
 }
 {bone1Name=dyn_hand_l, bone2Name=dyn_lowerarm_l, swing1LimitDegrees=10, swing2LimitDegrees=10, twistLimitDegrees=10}
 {bone1Name=dyn_upperarm_r, bone2Name=dyn_pelvis, swing1LimitDegrees=14, swing2LimitDegrees=14, twistLimitDegrees=10}
 {
  bone1Name=dyn_lowerarm_r
  bone2Name=dyn_upperarm_r
  swing1LimitDegrees=45
  swing1Motion=Locked
  swing2LimitDegrees=10
  twistLimitDegrees=45
  twistMotion=Locked
 }
 {bone1Name=dyn_hand_r, bone2Name=dyn_lowerarm_r, swing1LimitDegrees=10, swing2LimitDegrees=10, twistLimitDegrees=10}
 {bone1Name=dyn_tail_01, bone2Name=dyn_pelvis, swing1LimitDegrees=10, swing2LimitDegrees=10, twistLimitDegrees=10}
 {bone1Name=dyn_tail_02, bone2Name=dyn_tail_01, swing1LimitDegrees=7, swing2LimitDegrees=7, twistLimitDegrees=7}
 {bone1Name=dyn_tail_03, bone2Name=dyn_tail_02, swing1LimitDegrees=5, swing2LimitDegrees=5, twistLimitDegrees=5}
 {bone1Name=dyn_thigh_l, bone2Name=dyn_pelvis, swing1LimitDegrees=14, swing2LimitDegrees=14, twistLimitDegrees=10}
 {
  bone1Name=dyn_calf_l
  bone2Name=dyn_thigh_l
  swing1LimitDegrees=45
  swing1Motion=Locked
  swing2LimitDegrees=5
  twistLimitDegrees=45
  twistMotion=Locked
 }
 {bone1Name=dyn_foot_l, bone2Name=dyn_calf_l, swing1LimitDegrees=10, swing2LimitDegrees=10, twistLimitDegrees=10}
 {bone1Name=dyn_thigh_r, bone2Name=dyn_pelvis, swing1LimitDegrees=14, swing2LimitDegrees=14, twistLimitDegrees=10}
 {
  bone1Name=dyn_calf_r
  bone2Name=dyn_thigh_r
  swing1LimitDegrees=45
  swing1Motion=Locked
  swing2LimitDegrees=5
  twistLimitDegrees=45
  twistMotion=Locked
 }
 {bone1Name=dyn_foot_r, bone2Name=dyn_calf_r, swing1LimitDegrees=10, swing2LimitDegrees=10, twistLimitDegrees=10}
 {
  bone1Name=dyn_pelvis
  bone2Name=pack
  swing1LimitDegrees=45
  swing1Motion=Locked
  swing2LimitDegrees=45
  swing2Motion=Locked
  twistLimitDegrees=45
  twistMotion=Locked
 }
]
