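# Animation Blueprint for the EmberRae backpack, playing idles with nuzzle breaks and a burp reaction.
# Separate frontend and in-game state machines trigger burps on repeating delays or on chest, elimination and SlurpJuice shield events, gated by the part modifier's reactivity flag.
# Additive pack bone offsets follow spine yaw and crouching, then feed a curve-weighted RigidBody simulation for secondary motion.
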AnimBlueprint /BRCosmetics/Accessories/FORT_Backpacks/Backpack_F_MED_EmberRae/Meshes/F_MED_EmberRae_Pack_AnimBP
target_skeleton=&/F_MED_EmberRae_Pack_Skeleton
graphs [
 AnimGraph {nodes [
   AnimGraphNode_StateMachine_1 = Reactive Front End() {is=title AnimGraphNode_StateMachine}
   AnimGraphNode_StateMachine_2 = Idle and Reactive() {is=title AnimGraphNode_StateMachine}
   n1 = Get bShouldDisplayFrontendReactivity() -> (bShouldDisplayFrontendReactivity:bool)
   n2 = "Transform (Modify) Bone - Bone: pack"(
    ComponentPose=Local To Component(LocalPose=Blend Poses by bool(BlendPose_0=@AnimGraphNode_StateMachine_1, BlendPose_1=@AnimGraphNode_StateMachine_2, bActiveValue=@n1) {blendType=Linear})
    Translation=Get backpack_translation()
    Rotation=Get backpack_rotate()
   ) {rotationMode=BMM_Additive, translationMode=BMM_Additive, alphaBoolBlend {blendOption=Linear}, boneToModify {boneName=pack}}
   n3 = "Transform (Modify) Bone - Bone: pack"!(ComponentPose=@n2, bAlphaBoolEnabled:bool=Get bIsCrouching(), Translation="0,-2,0", Rotation="0,0,0", Scale) -> (Pose) {
    alphaInputType=Bool
    rotationMode=BMM_Additive
    translationMode=BMM_Additive
    alphaBoolBlend {blendInTime=0.1, blendOption=Linear, blendOutTime=0.1}
    boneToModify {boneName=pack}
    translation {y=-2}
   }
   n4 = RigidBody!(ComponentPose=@n3, AlphaCurveName:Name=PhysicsAlpha, ExternalForce) -> (Pose) {
    alphaInputType=Curve
    bOverrideWorldGravity=true
    overridePhysicsAsset=&/F_MED_EmberRae_Pack_Physics
    simulationSpace=BaseBoneSpace
    alphaBoolBlend {blendOption=Linear}
    baseBoneRef {boneName=root}
    componentAppliedLinearAccClamp=<2000 2000 2000>
    componentLinearAccScale=<0.4 0.4 0.4>
    componentLinearVelScale=<0.36 0.36 0.36>
    overrideWorldGravity {z=-980}
    simSpaceSettings {maxAngularAcceleration=2000, maxAngularVelocity=2000, maxLinearAcceleration=2000, maxLinearVelocity=2000}
   }
   Output Pose(Result=Component To Local(ComponentPose=@n4)) {layerGroup=None}
  ]}
 Reactive Front End {
  entry=FE_Idle
  path=AnimGraph.AnimGraphNode_StateMachine_1.Reactive Front End
  ids=id
  nodes [
   AnimStateNode_2 = FE_Idle() {
    is=title AnimStateNode
    stateEntered {guid=#1}
    stateFullyBlended {guid=#2}
    stateLeft {guid=#3}
   }
   AnimStateNode_3 = FE_Reactive_Anim() {
    is=title AnimStateNode
    stateEntered {guid=#4, notifyName=Entered FE_Reactive_Anim}
    stateFullyBlended {guid=#5}
    stateLeft {guid=#6}
   }
   AnimStateTransitionNode_2 = AnimStateTransitionNode() {
    crossfadeDuration=0.3
    transitionEnd {guid=#7}
    transitionInterrupt {guid=#8}
    transitionStart {guid=#9}
   }
   AnimStateTransitionNode_3 = AnimStateTransitionNode() {
    bAutomaticRuleBasedOnSequencePlayerInState=true
    crossfadeDuration=0.1
    transitionEnd {guid=#10}
    transitionInterrupt {guid=#11}
    transitionStart {guid=#12}
   }
  ]
  links [
   {
    dst=FE_Reactive_Anim
    rule=AnimGraph.AnimGraphNode_StateMachine_1.Reactive Front End.AnimStateTransitionNode_2.Transition
    src=FE_Idle
    via=AnimStateTransitionNode_2
   }
   {
    dst=FE_Idle
    rule=AnimGraph.AnimGraphNode_StateMachine_1.Reactive Front End.AnimStateTransitionNode_3.Transition
    src=FE_Reactive_Anim
    via=AnimStateTransitionNode_3
   }
  ]
 }
 FE_Idle {
  path=AnimGraph.AnimGraphNode_StateMachine_1.Reactive Front End.AnimStateNode_2.FE_Idle
  nodes [
   n1 = Random Sequence Player!() -> (Pose:Pose Link Structure) {
    is=title AnimGraphNode_RandomPlayer
    entries [
     {
      chanceToPlay=0.8
      maxLoopCount=0
      maxPlayRate=0.8
      minLoopCount=0
      minPlayRate=1
      sequence=/BRCosmetics/Accessories/FORT_Backpacks/Backpack_F_MED_EmberRae/Animations/EmberRae_Idle_PACK
      blendIn {blendOption=Linear, blendTime=0.2, customCurve=null}
     }
     {
      chanceToPlay=0.2
      maxLoopCount=0
      maxPlayRate=1
      minLoopCount=0
      minPlayRate=1
      sequence=/BRCosmetics/Accessories/FORT_Backpacks/Backpack_F_MED_EmberRae/Animations/EmberRae_IdleBreak_Nuzzle_PACK
      blendIn {blendOption=Linear, blendTime=0.2, customCurve=null}
     }
    ]
   }
   Output Animation Pose(Result=@n1)
  ]
 }
 FE_Reactive_Anim {
  path=AnimGraph.AnimGraphNode_StateMachine_1.Reactive Front End.AnimStateNode_3.FE_Reactive_Anim
  nodes [Output Animation Pose(Result=Sequence Player 'EmberRae_Burp_PACK'() {bLoopAnimation=false, sequence=/BRCosmetics/Accessories/FORT_Backpacks/Backpack_F_MED_EmberRae/Animations/EmberRae_Burp_PACK})]
 }
 Transition {
  path=AnimGraph.AnimGraphNode_StateMachine_1.Reactive Front End.AnimStateTransitionNode_2.Transition
  nodes [
   Result(bCanEnterTransition=Get Play Reactive Front End Anim() -> (Play Reactive Front End Anim:bool))
   Get ReactiveAnimation() -> (ReactiveAnimation:bool)
  ]
 }
 Transition {path=AnimGraph.AnimGraphNode_StateMachine_1.Reactive Front End.AnimStateTransitionNode_3.Transition, nodes [Result()]}
 Idle and Reactive {
  entry=Idle
  path=AnimGraph.AnimGraphNode_StateMachine_2.Idle and Reactive
  ids=id
  nodes [
   AnimStateNode_2 = Idle() {
    stateEntered {guid=#1}
    stateFullyBlended {guid=#2}
    stateLeft {guid=#3}
   }
   AnimStateNode_3 = Reactive_Anim() {
    is=title AnimStateNode
    stateEntered {guid=#4}
    stateFullyBlended {guid=#5}
    stateLeft {guid=#6}
   }
   AnimStateTransitionNode_2 = AnimStateTransitionNode() {
    crossfadeDuration=0.3
    transitionEnd {guid=#7}
    transitionInterrupt {guid=#8}
    transitionStart {guid=#9}
   }
   AnimStateTransitionNode_3 = AnimStateTransitionNode() {
    bAutomaticRuleBasedOnSequencePlayerInState=true
    crossfadeDuration=0.1
    transitionEnd {guid=#10}
    transitionInterrupt {guid=#11}
    transitionStart {guid=#12}
   }
  ]
  links [
   {
    dst=Reactive_Anim
    rule=AnimGraph.AnimGraphNode_StateMachine_2.Idle and Reactive.AnimStateTransitionNode_2.Transition
    src=Idle
    via=AnimStateTransitionNode_2
   }
   {
    dst=Idle
    rule=AnimGraph.AnimGraphNode_StateMachine_2.Idle and Reactive.AnimStateTransitionNode_3.Transition
    src=Reactive_Anim
    via=AnimStateTransitionNode_3
   }
  ]
 }
 Idle {
  path=AnimGraph.AnimGraphNode_StateMachine_2.Idle and Reactive.AnimStateNode_2.Idle
  nodes [
   n1 = Random Sequence Player() {entries [
     {
      chanceToPlay=0.8
      maxLoopCount=0
      maxPlayRate=0.8
      minLoopCount=0
      minPlayRate=1
      sequence=&EmberRae_Idle_PACK
      blendIn {blendOption=Linear, blendTime=0.2, customCurve=null}
     }
     {
      chanceToPlay=0.2
      maxLoopCount=0
      maxPlayRate=1
      minLoopCount=0
      minPlayRate=1
      sequence=&EmberRae_IdleBreak_Nuzzle_PACK
      blendIn {blendOption=Linear, blendTime=0.2, customCurve=null}
     }
    ]}
   Output Animation Pose(Result=@n1)
  ]
 }
 Reactive_Anim {
  path=AnimGraph.AnimGraphNode_StateMachine_2.Idle and Reactive.AnimStateNode_3.Reactive_Anim
  nodes [Output Animation Pose(Result=Sequence Player 'EmberRae_Burp_PACK'() {bLoopAnimation=false, sequence=&EmberRae_Burp_PACK})]
 }
 Transition {
  path=AnimGraph.AnimGraphNode_StateMachine_2.Idle and Reactive.AnimStateTransitionNode_2.Transition
  nodes [Result(bCanEnterTransition=Get ReactiveAnimation())]
 }
 Transition {path=AnimGraph.AnimGraphNode_StateMachine_2.Idle and Reactive.AnimStateTransitionNode_3.Transition, nodes [Result()]}
 EventGraph {nodes [
   Event OnCharacterCustomizationLoaded()
   > n1 = Cast To FortPlayerPawn(Object=TryGetFortPawn())
   > n2 = Cast To B_Athena_PartModifier_F_EmberRae_Backpack_Temp!(
    execute
    Object=GetModifierForPartType!(self:Fort Player Pawn Object Reference=@n1, PartType:EFortCustomPartType Enum=Backpack) -> (ReturnValue:Custom Character Part Modifier Object Reference)
   ) -> (then, CastFailed, AsB Athena Part Modifier F Ember Rae Backpack Temp:B Athena Part Modifier F Ember Rae Backpack Temp Object Reference) {is=title K2Node_DynamicCast}
   > n3 = Set Reactivity Enabled(Reactivity Enabled:bool=Get ReactivityEnabled!(self:B Athena Part Modifier F Ember Rae Backpack Temp Object Reference=@n2) -> (ReactivityEnabled:bool)) -> (Output_Get:bool)
   > n4 = Branch(Condition=@n3)
   > Do Once(Reset=@n6.then)
   > n5 = Delay(Duration=4)
   Set Play Reactive Front End Anim(execute=[@n14.then, @n5.then], Play Reactive Front End Anim:bool=true) -> (Output_Get:bool)
   n6 = Set Play Reactive Front End Anim(execute=@n4.else, Play Reactive Front End Anim=false)
   Event BlueprintUpdateAnimation()
   n7 = GetOwnerSkeletalMeshComponent()
   n8 = BreakRotator(InRot=BreakTransform(InTransform=GetDeltaTransformFromRefPose(self=@n7, BoneName=spine_05, BaseName=spine_01)).Rotation)
   > Set backpack_rotate(backpack_rotate=MakeRotator(Roll=MapRangeClamped(Value=@n8.Yaw, InRangeA=0, InRangeB=-30, OutRangeB=35)))
   > Set backpack_translation(backpack_translation=MakeVector(Y=MapRangeClamped(Value=@n8.Yaw, InRangeA=0, InRangeB=-30, OutRangeB=-6)))
   n9 = Event BlueprintProcessAnimInputEvent()
   n10 = IsNamed(self=@n9.AnimInputEvent, InEventName=aie.ChestOpened)
   n11 = IsNamed(self=@n9.AnimInputEvent, InEventName=aie.EnemyPlayerEliminated)
   n12 = Get CurrentWeapon!(self:Fort Pawn Object Reference=Get As Player Pawn Athena() -> (As Player Pawn Athena:Player Pawn Athena Object Reference)) -> (CurrentWeapon:Fort Weapon Object Reference)
   n13 = AND Boolean(
    A=IsNamed(self=@n9.AnimInputEvent, InEventName=aie.PlayerVitality.ShieldsIncreased)
    B=Contains(SearchIn=GetDisplayName(Object=@n12), Substring=SlurpJuice)
   )
   > Branch(Condition=AND Boolean(A=OR Boolean!(A=@n10, B=@n11, C:bool=@n13) -> (ReturnValue), B=Get Reactivity Enabled() -> (Reactivity Enabled:bool)))
   > Set ReactiveAnimation(ReactiveAnimation:bool=true) -> (Output_Get:bool)
   Event BlueprintPostEvaluateAnimation() {is=title K2Node_Event}
   > Set ReactiveAnimation(ReactiveAnimation=false)
   AnimNotify_Entered FE_Reactive_Anim() {is=title K2Node_Event}
   > Set Play Reactive Front End Anim(Play Reactive Front End Anim=false)
   > n14 = Delay(Duration=8)
   TryGetPawnOwner()
  ]}
]
variables [
 As Player Pawn Athena {type=PlayerPawn_Athena_C}
 backpack_rotate {type=Rotator}
 backpack_translation {type=Vector}
 Play Reactive Front End Anim {type=boolean}
 ReactiveAnimation {type=boolean}
 Reactivity Enabled {type=boolean}
]
